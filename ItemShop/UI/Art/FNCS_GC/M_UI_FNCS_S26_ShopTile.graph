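# UI material for the FNCS S26 Item Shop tile, producing an animated, shimmering sticker-texture background.
# The sticker texture is distorted by a panning UI warp, and two flag-texture noise samples multiply into a Shine-scaled highlight added on top.
# Emissive_Intensity scales that result before a final blend; a fill-scaled UV and a warm exponential radial gradient are declared as named reroutes.
# An Icon texture parameter is present with no visible connection.
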Material /ItemShop/UI/Art/FNCS_GC/M_UI_FNCS_S26_ShopTile
graph {
 properties {
  bCastRayTracedShadows=false
  bUsedWithNanite=true
  bUsedWithStaticMesh=false
  materialDomain=MD_UI
  pixelDepthOffsetMode=PDOM_Legacy
  shadingModel=MSM_Unlit
  usedShadingModels=Unlit
 }
 nodes [
  n1 = NamedRerouteUsage()
  n2 = LinearInterpolate() {constA=0.05, constAlpha=0, constB=0.08}
  n3 = TextureSample(UVs=MaterialFunctionCall!(UV=@n1, Warp Center, Warp Intensity=Panner(Coordinate=@n1, Speed=@n2) {speedX=0.02, speedY=0.02}, Warp Radius) {materialFunction=/FortUILibrary/MaterialFunctions/Effects/MF_UI_Warp}.Warped UV) {texture=/OfferCatalog/Art/A_Shop_Tiles_Textures/FNCS_S26/T-UI-Sticker-Texture}
  n4 = NamedRerouteUsage()
  n5 = LinearInterpolate(
   A=TextureSample(UVs=NamedRerouteUsage()) {samplerType=SAMPLERTYPE_Alpha, texture=/OfferCatalog/Art/FNCS22/T_UI_FlagTexture}.R
   B=Multiply(A=TextureCoordinate())
   Alpha=Constant() {r=1}
  )
  n6 = TextureSample(UVs=Max(A=Panner(Coordinate=@n4) {speedX=0.02, speedY=0.05}, B=@n5)) {samplerType=SAMPLERTYPE_Alpha, texture=&T_UI_FlagTexture}
  n7 = TextureSample(
   UVs=@n5
   Tex=TextureObjectParameter() {parameterName=Noise, samplerType=SAMPLERTYPE_Alpha, texture=&T_UI_FlagTexture}
  )
  n8 = Multiply(A=@n3.RGB, B=Multiply(A=Multiply(A=@n6.RGB, B=@n7.RGB), B=ScalarParameter() {defaultValue=1, parameterName=Shine})) {constB=4}
  LinearInterpolate(
   A=Multiply(A=Add(A=@n3.RGB, B=@n8), B=ScalarParameter() {defaultValue=0.7, parameterName=Emissive_Intensity})
   B=NamedRerouteUsage()
   Alpha=NamedRerouteUsage()
  )
  NamedRerouteDeclaration(None=Multiply(A=MaterialFunctionCall() {materialFunction=&MF_UI_UVScaling}.Scaled To Fill)) {name=UVFill, nodeColor=<rgba 1 0 0.7266 1>}
  n9 = MaterialFunctionCall(UVs=NamedRerouteUsage(), Radius=Constant() {r=1.4}, Density=Constant() {r=1.4}) {materialFunction=&RadialGradientExponential}
  NamedRerouteDeclaration(None=Multiply(A=Constant3Vector() {constant=<rgba 0.9911 0.8796 0.5841 1>}, B=@n9.RadialGradientExponential)) {name=Gradient, nodeColor=<rgba 0.3828 0 1 1>}
  NamedRerouteDeclaration(None=@n9.RadialGradientExponential) {name=Gradient Alpha, nodeColor=<rgba 1 0 0.0937 1>}
  TextureSampleParameter2D() {parameterName=Icon}
 ]
}
