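# UI material for the FNCS item shop background, drawing animated blue water framed by slate corner art.
# A panned flow-map sample, lerped with tiled UVs, offsets rotated UVs to distort a waterfall texture that blends two blues.
# A top-bottom gradient fades the water toward flat blue, then the SlateCorners green channel composites it with the pale-tinted red channel.
# MF_UI_UVScaling provides 1080p-based fit and fill UVs through named reroutes.
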Material /ItemShop/UI/Art/FNCS_GC/M_UI_FNCS_ShopBG
graph {
 properties {bUsedWithStaticMesh=false, materialDomain=MD_UI, pixelDepthOffsetMode=PDOM_Legacy, usedShadingModels=Default Lit}
 nodes [
  Constant3Vector() {constant=<rgba 0.0024 0.0331 0.2747 1>}
  Constant3Vector() {constant=<rgba 0.004 0.0497 0.5457 1>}
  n1 = TextureSample(UVs=NamedRerouteUsage()) {samplerType=SAMPLERTYPE_LinearColor, texture=&/T_UI_SlateCorners}
  n2 = Constant3Vector() {constant=<rgba 0.0058 0.0378 0.2882 1>}
  n3 = Panner(Coordinate=Multiply(A=NamedRerouteUsage()) {constB=3}) {speedX=-0.05, speedY=-0.07}
  n4 = TextureCoordinate()
  n5 = Add(
   A=MaterialFunctionCall(UVs=NamedRerouteUsage(), Rotation Center=Constant() {r=0.5}, "Rotation Angle (0-1)"=Constant() {r=-0.7}) {materialFunction=&CustomRotator}.Rotated Values
   B=LinearInterpolate(
    A=TextureSample(UVs=@n3) {samplerType=SAMPLERTYPE_LinearColor, texture=/Game/UI/Frontend/ItemShop/Art/WaterBackground/T_UI_WaterFlowMap}.R
    B=Multiply(A=@n4) {constB=3}
    Alpha=Constant() {r=1.25}
   )
  )
  n6 = LinearInterpolate(
   A=@n2
   B=Constant3Vector() {constant=<rgba 0.0156 0.0621 0.4271 1>}
   Alpha=OneMinus(None=TextureSample(UVs=@n5) {texture=&/T_UI_FNCS_Waterfall}.R)
  )
  n7 = MaterialFunctionCall!(UVs=NamedRerouteUsage(), Subtract=Constant() {r=0.25}, Falloff=Constant() {r=0.7}, Intensity=Constant() {r=2}) {materialFunction=/Game/UI/Frontend/ItemShop/Art/_MaterialFunctions_ShopTheming/MF_UI_TopBottomGradient}
  LinearInterpolate(
   A=Multiply(A=Constant3Vector() {constant=<rgba 0.5974 0.6964 0.7812 1>}, B=@n1.R)
   B=Add(A=LinearInterpolate(A=@n6, B=@n2, Alpha=@n7.Result) {constB=0}) {constB=0.05}
   Alpha=@n1.G
  )
  n8 = MaterialFunctionCall(TextureSize=Constant2Vector() {g=1080, r=1920}) {materialFunction=&MF_UI_UVScaling}
  NamedRerouteDeclaration(None=@n8.Scaled to Fit) {name=UVFit, nodeColor=<rgba 0 0.1094 1 1>}
  NamedRerouteDeclaration(None=@n8.Scaled To Fill) {name=UVFill, nodeColor=<rgba 0 1 0.2734 1>}
 ]
}
